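; Blueprint function library for the Stamina "Hawaii" beam item, covering aiming, beam tracing and player-versus-player beam battles.
; Beam endpoints extend over time at a scalable speed until a line trace hits or reaches max range.
; CheckForBeamBattle pairs active beams by segment proximity and facing angle, using state on each pawn's Hawaii component.
; Other helpers find reticle targets, randomize decal rotation and spawn directional victim-warning lens effects.
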(fn CacheBeamEndpoint
  (params (BeamShooter :type Fort%20Player%20Pawn%20Object%20Reference) (Endpoint :type Vector))
  (bind _returnvalue (Actor|GetComponentbyClass BeamShooter "/StaminaGameplay/Gameplay/Items/Hawaii/Stamina_Hawaii_Component.Stamina_Hawaii_Component_C"))
  (Utilities|IsValid _returnvalue
    (:Is%20Valid
      (Variables|/StaminaGameplay/Gameplay/Items/Hawaii/Stamina_Hawaii_Component.Stamina_Hawaii_Component_C|SetRepBattleIntersectionPoint Endpoint _returnvalue))))

(fn CheckForBeamBattle
  (params (PawnShooter :type Fort%20Player%20Pawn%20Object%20Reference) (ShooterEndpoint :type Vector) (AngleThreshold :type Float%20%28double-precision%29) (DistanceThreshold :type Float%20%28double-precision%29))
  (bind (_segment1point _segment2point) (Math|Vector|FindNearestPointsonLineSegments (LocalVariables|GetShooterBeamOrigin) (LocalVariables|GetShooterBeamEndpoint) (LocalVariables|GetTargetBeamOrigin) (LocalVariables|GetTargetBeamEndpoint)))
  (bind _possibletarget (LocalVariables|GetPossibleTarget))
  (bind _outorigin (CallFunction|GetBeamOriginforBeamBattle PawnShooter))
  (LocalVariables|SetShooterBeamOrigin _outorigin)
  (LocalVariables|SetDistThresholdSqr (* DistanceThreshold DistanceThreshold))
  (bind _output_get (LocalVariables|SetShooterBeamEndpoint ShooterEndpoint))
  (LocalVariables|SetShooterBeamRange (Math|Vector|VectorLength (- _output_get (LocalVariables|GetShooterBeamOrigin))))
  (LocalVariables|SetAngle AngleThreshold)
  (LocalVariables|SetShooter PawnShooter)
  (bind _outfortplayerpawns (Fortnite|GetFortPlayerPawnsInRange (LocalVariables|GetShooter) (LocalVariables|GetShooterBeamOrigin) 2e+04))
  (for _array_element _outfortplayerpawns
    (if (not (LocalVariables|GetFoundTarget))
      (LocalVariables|SetPossibleTarget _array_element)
      (if (not (== (LocalVariables|GetPossibleTarget) (LocalVariables|GetShooter)))
        (bind _inanotherbeambattle (Default|IsinBeamBattlewithOtherPlayer _possibletarget (LocalVariables|GetShooter)))
        (if (and (GameplayTags|HasMatchingGameplayTag (Ability|GetAbilitySystemComponent _possibletarget) "(TagName=\"Stamina.Hawaii.BeamActive\")") (not _inanotherbeambattle))
          (bind _outendpoint (CallFunction|GetBeamEndpoint (LocalVariables|GetPossibleTarget)))
          (LocalVariables|SetTargetBeamEndpoint _outendpoint)
          (bind _outorigin_2 (CallFunction|GetBeamOriginforBeamBattle (LocalVariables|GetPossibleTarget)))
          (LocalVariables|SetTargetBeamOrigin _outorigin_2)
          (if (< (Math|Vector|DistanceSquared(Vector) _segment1point _segment2point) (* (LocalVariables|GetDistThresholdSqr) (LocalVariables|GetDistThresholdSqr)))
            (LocalVariables|SetShooterBeamDirection (- (LocalVariables|GetShooterBeamEndpoint) (LocalVariables|GetShooterBeamOrigin)))
            (LocalVariables|SetTargetBeamDirection (- (LocalVariables|GetTargetBeamEndpoint) (LocalVariables|GetTargetBeamOrigin)))
            (LocalVariables|SetShooterOffsetToTarget (- (LocalVariables|GetTargetBeamOrigin) (LocalVariables|GetShooterBeamOrigin)))
            (LocalVariables|SetTargetOffsetToShooter (- (LocalVariables|GetShooterBeamOrigin) (LocalVariables|GetTargetBeamOrigin)))
            (bind _returnvalue (Fortnite|GetAngleDegrees (LocalVariables|GetShooterBeamDirection) (LocalVariables|GetShooterOffsetToTarget)))
            (bind _returnvalue_2 (Fortnite|GetAngleDegrees (LocalVariables|GetTargetBeamDirection) (LocalVariables|GetTargetOffsetToShooter)))
            (if (and (< _returnvalue_2 (LocalVariables|GetAngle)) (< _returnvalue (LocalVariables|GetAngle)))
              (LocalVariables|SetFoundTarget true)
              (LocalVariables|SetFinalTarget (LocalVariables|GetPossibleTarget))
              (LocalVariables|SetIntersectPoint (/ (+ (LocalVariables|GetShooterBeamOrigin) (LocalVariables|GetTargetBeamOrigin)) 2.0))))))))
  (return (LocalVariables|GetFoundTarget) (LocalVariables|GetFinalTarget) (LocalVariables|GetIntersectPoint)))

(fn ComputeBeamTraceParams
  (params (Pawn :type Fort%20Player%20Pawn%20Object%20Reference) (SocketName :type Name) (MaxRange :type Float%20%28double-precision%29))
  (bind _cachedpawn (LocalVariables|GetCachedPawn))
  (LocalVariables|SetCachedPawn Pawn)
  (if (Utilities|Enum|Equal(Enum) (Networking|GetLocalRole (LocalVariables|GetCachedPawn)) "ROLE_SimulatedProxy")
    (LocalVariables|SetAimRotation (Pawn|GetBaseAimRotation (LocalVariables|GetCachedPawn)))
    (else
      (LocalVariables|SetAimRotation (Pawn|GetControlRotation (LocalVariables|GetCachedPawn) :!class /Script/Engine.Pawn))))
  (LocalVariables|SetCameraForward (Math|Vector|GetForwardVector (LocalVariables|GetAimRotation)))
  (LocalVariables|SetPawnForward (Math|Vector|GetForwardVector (Transformation|GetActorRotation _cachedpawn)))
  (LocalVariables|SetTraceStart (Transformation|GetSocketLocation (Variables|/Script/Engine.Character|GetMesh _cachedpawn) SocketName))
  (LocalVariables|SetProjectedAimPoint (+ (LocalVariables|GetTraceStart) (* (Math|Vector|GetForwardVector (Math|Rotator|MakeRotator (.roll (Transformation|GetActorRotation (LocalVariables|GetCachedPawn))) (.pitch (LocalVariables|GetAimRotation)) (.yaw (Transformation|GetActorRotation (LocalVariables|GetCachedPawn))))) MaxRange)))
  (return (LocalVariables|GetTraceStart) (LocalVariables|GetProjectedAimPoint)))

(fn ComputeRelativeViewLocation
  (params (LeftRightOffset :type Float%20%28double-precision%29) (ForwardBackOffset :type Float%20%28double-precision%29) (UpDownOffset :type Float%20%28double-precision%29) (InLocation :type Vector) (InViewRotation :type Rotator))
  (LocalVariables|SetLocalLeftRight LeftRightOffset)
  (LocalVariables|SetLocalFrontBack ForwardBackOffset)
  (LocalVariables|SetLocalUpDown UpDownOffset)
  (LocalVariables|SetLocalInLocation InLocation)
  (LocalVariables|SetLocalInRotation InViewRotation)
  (LocalVariables|SetFinalOutLocation (+ (+ (+ (LocalVariables|GetLocalInLocation) (* (Math|Vector|GetRightVector (LocalVariables|GetLocalInRotation)) (LocalVariables|GetLocalLeftRight))) (* (Math|Vector|GetForwardVector (LocalVariables|GetLocalInRotation)) (LocalVariables|GetLocalFrontBack))) (* (Math|Vector|GetUpVector (LocalVariables|GetLocalInRotation)) (LocalVariables|GetLocalUpDown))))
  (return (LocalVariables|GetFinalOutLocation)))

(fn GetBeamBattleIntersectionPoint
  (params (Pawn :type Fort%20Player%20Pawn%20Object%20Reference) (SocketName :type Name))
  (bind _returnvalue (Actor|GetComponentbyClass Pawn "/StaminaGameplay/Gameplay/Items/Hawaii/Stamina_Hawaii_Component.Stamina_Hawaii_Component_C"))
  (bind _targetend (LocalVariables|GetTargetEnd))
  (bind _targetstart (LocalVariables|GetTargetStart))
  (bind _shooterend (LocalVariables|GetShooterEnd))
  (bind _shooterstart (LocalVariables|GetShooterStart))
  (bind (_segment1point _segment2point) (Math|Vector|FindNearestPointsonLineSegments _shooterstart _shooterend _targetstart _targetend))
  (bind _returnvalue_2 (+ _segment1point _segment2point))
  (bind _returnvalue_3 (/ _returnvalue_2 2.0))
  (bind _hawaiicomp (LocalVariables|GetHawaiiComp))
  (bind _repbeambattleopponent (Variables|/StaminaGameplay/Gameplay/Items/Hawaii/Stamina_Hawaii_Component.Stamina_Hawaii_Component_C|GetRepBeamBattleOpponent _hawaiicomp))
  (LocalVariables|SetSocket SocketName)
  (Utilities|IsValid _returnvalue
    (:Is%20Valid
      (LocalVariables|SetHawaiiComp _returnvalue)
      (bind (_outtracestart _outtraceend) (Class|StaminaFunctionLibraryNative|ComputeBeamTraceParams Pawn (LocalVariables|GetSocket) 5e+04))
      (LocalVariables|SetShooterStart _outtracestart)
      (LocalVariables|SetShooterEnd _outtraceend)
      (Utilities|IsValid _repbeambattleopponent
        (:Is%20Valid
          (bind (_outtracestart_2 _outtraceend_2) (Class|StaminaFunctionLibraryNative|ComputeBeamTraceParams _repbeambattleopponent (LocalVariables|GetSocket) 5e+04))
          (LocalVariables|SetTargetStart _outtracestart_2)
          (LocalVariables|SetTargetEnd _outtraceend_2)
          (Rendering|Debug|DrawDebugSphere 0 _returnvalue_3 1e+02 12 "(R=1.000000,G=1.000000,B=1.000000,A=1.000000)" 0.1)
          (return _returnvalue_3 (LocalVariables|GetTargetStart)))))))

(fn GetBeamEndpoint
  (params (BeamShooter :type Fort%20Player%20Pawn%20Object%20Reference))
  (bind _returnvalue (Actor|GetComponentbyClass BeamShooter "/StaminaGameplay/Gameplay/Items/Hawaii/Stamina_Hawaii_Component.Stamina_Hawaii_Component_C"))
  (Utilities|IsValid _returnvalue
    (:Is%20Valid
      (return (Variables|/StaminaGameplay/Gameplay/Items/Hawaii/Stamina_Hawaii_Component.Stamina_Hawaii_Component_C|GetServerBeamEndpoint _returnvalue)))
    (:Is%20Not%20Valid
      (Development|PrintString 0 "NoValidStaminaHawaiiComponent")
      (return "0, 0, 0"))))

(fn GetBeamOriginForBeamBattle
  (params (Pawn :type Fort%20Player%20Pawn%20Object%20Reference))
  ;[1 1 1] Not Literally where the beam is attached, but offset to make beam battle look better
  (Utilities|IsValid Pawn
    (:Is%20Valid
      (LocalVariables|SetOrigin (+ (Transformation|GetActorLocation Pawn) (* (Transformation|GetActorUpVector Pawn) 1e+02))))
    (:Is%20Not%20Valid
      _))
  (return (LocalVariables|GetOrigin)))

(fn GetPlayerPossibleTargets
  (params (Shooter :type Fort%20Player%20Pawn%20Object%20Reference) (ReticleSize :type Float%20%28double-precision%29) (Range :type Float%20%28double-precision%29))
  (bind _outfortplayerpawns (Fortnite|GetFortPlayerPawnsInRange 0 (Transformation|GetActorLocation (Variables|GetShooter)) (Variables|GetRange)))
  (for _array_element _outfortplayerpawns
    (LocalVariables|SetLoopBodyVictim _array_element)
    (if (Team|OnSameTeam (LocalVariables|GetLoopBodyVictim) (Variables|GetShooter))
      _
      (elif (> (Math|Vector|DotProduct (Math|Vector|GetForwardVector (Pawn|GetBaseAimRotation (Variables|GetShooter))) (- (Transformation|GetActorLocation (LocalVariables|GetLoopBodyVictim)) (Transformation|GetActorLocation (Variables|GetShooter)))) 0)
        (bind _outrelativeaimpoint (CallFunction|GetPlayerRelativeAimpoint (Variables|GetShooter) (LocalVariables|GetLoopBodyVictim)))
        (if (<= _outrelativeaimpoint (Variables|GetReticleSize))
          (Utilities|Array|Add (LocalVariables|GetTargets) (LocalVariables|GetLoopBodyVictim))))))
  (return (LocalVariables|GetTargets)))

(fn GetPlayerRelativeAimpoint
  (params (Shooter :type Fort%20Player%20Pawn%20Object%20Reference) (Victim :type Fort%20Player%20Pawn%20Object%20Reference))
  ;[1 1 1] Compute the horizontal/vertical offset of the shooters aimpoint relative to the target pawn
  (bind _asfort_player_controller (Utilities|Casting|CastToFortPlayerController (Pawn|GetController (Variables|GetShooter))))
  (LocalVariables|SetViewpoint (Pawn|GetPlayerViewPoint _asfort_player_controller))
  (LocalVariables|SetAimRotation (.Rotation (Pawn|GetPlayerViewPoint _asfort_player_controller)))
  (LocalVariables|SetLineEnd (+ (LocalVariables|GetViewpoint) (* (Math|Vector|GetForwardVector (LocalVariables|GetAimRotation)) 5e+04)))
  (return (Math|Vector|GetPointDistancetoSegment (Transformation|GetActorLocation (Variables|GetVictim)) (LocalVariables|GetViewpoint) (LocalVariables|GetLineEnd))))

(fn IsInBeamBattleWithOtherPlayer
  (params (OtherGuy :type Fort%20Player%20Pawn%20Object%20Reference) (MyGuy :type Fort%20Player%20Pawn%20Object%20Reference))
  (bind _returnvalue (Actor|GetComponentbyClass OtherGuy "/StaminaGameplay/Gameplay/Items/Hawaii/Stamina_Hawaii_Component.Stamina_Hawaii_Component_C"))
  (Utilities|IsValid _returnvalue
    (:Is%20Valid
      (return (and (Variables|/StaminaGameplay/Gameplay/Items/Hawaii/Stamina_Hawaii_Component.Stamina_Hawaii_Component_C|GetRepInBeamBattle _returnvalue) (!= (Variables|/StaminaGameplay/Gameplay/Items/Hawaii/Stamina_Hawaii_Component.Stamina_Hawaii_Component_C|GetRepBeamBattleOpponent _returnvalue) MyGuy))))
    (:Is%20Not%20Valid
      (return false))))

(fn RandomizeDecalRotation
  (params (InTarget :type Actor%20Object%20Reference))
  (bind _target (LocalVariables|GetTarget))
  (bind _returnvalue (Transformation|GetActorRotation _target))
  (LocalVariables|SetTarget InTarget)
  (LocalVariables|SetForward (Math|Vector|GetForwardVector _returnvalue))
  (LocalVariables|SetRight (Math|Vector|GetRightVector _returnvalue))
  (LocalVariables|SetUp (Math|Vector|GetUpVector _returnvalue))
  (LocalVariables|SetAngle (Math|Random|RandomFloatinRange -9e+01 9e+01))
  (Transformation|SetActorRotation (LocalVariables|GetTarget) (Math|Rotator|MakeRotationfromAxes (Math|Vector|RotateVectorAroundAxis (LocalVariables|GetForward) (LocalVariables|GetAngle) (LocalVariables|GetUp)) (Math|Vector|RotateVectorAroundAxis (LocalVariables|GetRight) (LocalVariables|GetAngle) (LocalVariables|GetUp)) (LocalVariables|GetUp)) false))

(fn SpawnVictimWarningLensEffect
  (params (Target :type Actor%20Object%20Reference) (Shooter :type Actor%20Object%20Reference))
  (bind _victimloc (LocalVariables|GetVictimLoc))
  (bind _shooterloc (LocalVariables|GetShooterLoc))
  (bind _returnvalue (- _shooterloc _victimloc))
  (bind _returnvalue_2 (Math|Vector|Normalize _returnvalue))
  (bind _returnvalue_3 (Math|Rotator|MakeRotfromX _returnvalue_2))
  (bind _yaw (.yaw _returnvalue_3))
  (bind _victimrot (LocalVariables|GetVictimRot))
  (bind _yaw_2 (.yaw _victimrot))
  (bind _returnvalue_4 (- _yaw _yaw_2))
  (LocalVariables|SetVictimLoc (Transformation|GetActorLocation Target))
  (LocalVariables|SetShooterLoc (Transformation|GetActorLocation Shooter))
  (bind _asfort_player_pawn (Utilities|Casting|CastToFortPlayerPawn Target))
  (LocalVariables|SetVictimRot (Pawn|GetControlRotation _asfort_player_pawn :!class /Script/Engine.Pawn))
  (bind _asfort_player_controller (Utilities|Casting|CastToFortPlayerController (Pawn|GetController _asfort_player_pawn)))
  (bind _returnvalue_5 (Camera|AddGenericCameraLensEffect (Variables|/Script/Engine.PlayerController|GetPlayerCameraManager _asfort_player_controller) "/StaminaGameplay/Gameplay/Items/Hawaii/PROTOTYPE/FX/BP_HawaiiStamina_CameraLensFX.BP_HawaiiStamina_CameraLensFX_C"))
  (for _array_element (CameraLensEffect|GetParticleComponents _returnvalue_5)
    (if (Math|Float|InRange(Float) _returnvalue_4 -45.0 45.0)
      (LocalVariables|SetUseDirection (LocalVariables|GetDirectionUp))
      (elif (Math|Float|InRange(Float) _returnvalue_4 45.0 135.0)
        (LocalVariables|SetUseDirection (LocalVariables|GetDirectionRight))
        (elif (Math|Float|InRange(Float) _returnvalue_4 -135.0 -45.0)
          (LocalVariables|SetUseDirection (LocalVariables|GetDirectionLeft))
          (Effects|Components|ParticleSystem|SetIntParameter _array_element "Direction" (LocalVariables|GetUseDirection))
          (else
            (LocalVariables|SetUseDirection (LocalVariables|GetDirectionDown))))))))

(fn TraceForBeamEndpoint
  (params (TargetPawn :type Fort%20Player%20Pawn%20Object%20Reference) (MaxRange :type Float%20%28double-precision%29) (SocketName :type Name) (StartTime :type Float%20%28double-precision%29) (Aimpoint :type Vector))
  (bind _hitresult (LocalVariables|GetHitResult))
  (bind (_bblockinghit _binitialoverlap _time _distance _location _impactpoint _normal _impactnormal) (Collision|BreakHitResult _hitresult))
  (LocalVariables|SetCurrentEndpointTravelDistance (Math|Float|Clamp(Float) (* (- (Utilities|Time|GetTimeSeconds 0) (Variables|GetStartTime)) (ScalableFloat|GetValueatLevel (LocalVariables|GetRow_BeamSpeed))) 0.0 (Variables|GetMaxRange)))
  (bind _outtracestart (CallFunction|ComputeBeamTraceParams (Variables|GetTargetPawn) (Variables|GetSocketName) (Variables|GetMaxRange)))
  (LocalVariables|SetTraceStart _outtracestart)
  (LocalVariables|SetBeamDirection (Math|Vector|Normalize (- (Variables|GetAimpoint) (LocalVariables|GetTraceStart))))
  (LocalVariables|SetTraceEnd (+ (LocalVariables|GetTraceStart) (* (LocalVariables|GetBeamDirection) (LocalVariables|GetCurrentEndpointTravelDistance))))
  (bind (_outhit _returnvalue) (Collision|LineTraceByProfile 0 (LocalVariables|GetTraceStart) (LocalVariables|GetTraceEnd) "FortTrajectoryBlocksWater" false (Utilities|Array|MakeArray (Variables|GetTargetPawn))))
  (LocalVariables|SetImpacted _returnvalue)
  (LocalVariables|SetHitResult _outhit)
  (if (LocalVariables|GetImpacted)
    (LocalVariables|SetFinalEndpoint _location)
    (else
      (LocalVariables|SetFinalEndpoint (LocalVariables|GetTraceEnd))))
  (return (LocalVariables|GetFinalEndpoint) (LocalVariables|GetImpacted) (>= (LocalVariables|GetCurrentEndpointTravelDistance) (Variables|GetMaxRange)) _impactnormal))
